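# Material function for the CarrotCake backpack that adds animated holographic scanlines and a center face glow to a material's emissive color.
# A sphere-mask glow, dimmed where the incoming emissive is near black, is added, then desaturated and hue-shifted.
# It is multiplied by main and top/bottom scanline layers, blended using a mask texture and pre-skinned normals.
# A final hue shift feeds the emissive output; a time-of-day tint branch is left unconnected.
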MaterialFunction /BRCosmetics/Accessories/FORT_Backpacks/Backpack_M_MED_CarrotCake/Materials/MF_CarrotCake_Backpack_Scanlines
graph {nodes [
  n1 = GetMaterialAttributes(None=FunctionInput() {inputName=MaterialAttributes, inputType=FunctionInput_MaterialAttributes}) {attributeGetTypes [#1]}
  n2 = ScalarParameter() {group=CarrotCake, parameterName=OverallHueShift, sortPriority=0}
  n3 = NamedRerouteUsage()
  n4 = Constant()
  n5 = OneMinus(None=Multiply(A=SphereMask(A=@n1.EmissiveColor, B=@n4) {attenuationRadius=0.05, hardnessPercent=20}) {constB=0.85})
  n6 = Desaturation(
   None=Add(A=Multiply(A=@n3, B=@n5), B=@n1.EmissiveColor)
   Fraction=ScalarParameter() {group=CarrotCake, parameterName=OverallDesaturation, sortPriority=0}
  )
  n7 = WorldPosition()
  n8 = Multiply(A=ComponentMask(None=TransformPosition(None=@n7.XYZ) {transformSourceType=TRANSFORMPOSSOURCE_World}) {b=true}) {constB=0.05}
  n9 = Multiply(A=TextureCoordinate()) {constB=2.4}
  n10 = NamedRerouteUsage()
  n11 = LinearInterpolate(
   A=ScalarParameter() {defaultValue=0.45, group=HolographicTexture, parameterName=ScanLineMinColor_Main, sortPriority=100}
   Alpha=MaterialFunctionCall!(
    ScanLine_Tiling_1
    Scanline_PannerSpeed_1
    ScanLine_Tiling
    Scanline_PannerSpeed
    Scanline_Rotation
    UVs=LinearInterpolate(A=@n8, B=@n9, Alpha=@n10)
   ) {materialFunction=/Game/Global/Materials/MaterialGym/Materials/HolographicTexture/MF_HolographicTexture_ScanLines}.Result
  ) {constA=0.45}
  n12 = ComponentMask(None=VertexInterpolator(VS=PreSkinnedPosition()).PS) {g=true, r=true}
  n13 = ScalarParameter() {defaultValue=0.05, group=HolographicTexture, parameterName="ScanLineSize_Top&&Bottom", sortPriority=100}
  n14 = LinearInterpolate(
   A=ScalarParameter() {defaultValue=0.55, group=HolographicTexture, parameterName="ScanLineMinColor_Top&&Bottom", sortPriority=100}
   Alpha=MaterialFunctionCall(Scanline_Rotation=Constant() {r=6.2832}, UVs=Multiply(A=@n12, B=@n13) {constB=0.05}) {materialFunction=&MF_HolographicTexture_ScanLines}.Result
  ) {constA=0.55}
  n15 = PreSkinnedNormal()
  n16 = Abs(None=ComponentMask(None=VertexInterpolator(VS=@n15).PS) {b=true})
  n17 = Multiply(A=NamedRerouteUsage(), B=Power(Base=Saturate(None=@n16)))
  n18 = MaterialFunctionCall(
   Hue Shift Percentage=ScalarParameter() {group=HolographicTexture, parameterName=Emissive_HueShift, sortPriority=101}
   Texture=Multiply(
    A=MaterialFunctionCall(Hue Shift Percentage=@n2, Texture=@n6) {materialFunction=&HueShift}.Result
    B=LinearInterpolate(A=@n11, B=@n14, Alpha=@n17) {constB=0.8}
   )
  ) {materialFunction=&HueShift}
  FunctionOutput(None=SetMaterialAttributes!(MaterialAttributes=@n1.MaterialAttributes, EmissiveColor=@n18.Result) {attributeSetTypes [#1]}) {outputName=Results}
  n19 = Constant() {r=1}
  Multiply(
   A=@n18.Result
   B=MaterialFunctionCall(Color_Morning=@n19, Color_Day=@n19, Color_Evening=@n19, Color_Night=Constant3Vector() {constant=<rgba 0.5153 0.557 0.85 1>}) {materialFunction=&Color_Time_of_Day}.Result
  )
  n20 = TextureSample() {
   samplerSource=SSM_Wrap_WorldGroupSettings
   samplerType=SAMPLERTYPE_Masks
   texture=/BRCosmetics/Accessories/FORT_Backpacks/Backpack_M_MED_CarrotCake/Textures/T_CarrotCake_Backpack_FaceMask_Test
  }
  NamedRerouteDeclaration(None=@n20.R) {name="MaskShoes(R)", nodeColor=<rgba 1 0 0.1577 1>}
  NamedRerouteDeclaration(None=@n20.G) {name="MaskTop&&Bottom(G)", nodeColor=<rgba 1 0.5832 0 1>}
  n21 = Power(
   Base=SphereMask(
    A=MaterialFunctionCall() {materialFunction=&LocalPosition}.Local Position
    B=Constant3Vector() {constant=<rgba 0 -10 3 0>}
    Radius=ScalarParameter() {defaultValue=17, group=CarrotCake, parameterName=CenterGlow_Radius}
   ) {attenuationRadius=17, hardnessPercent=0}
   Exp=ScalarParameter() {defaultValue=3, group=CarrotCake, parameterName=CenterGlow_FalloffPower}
  ) {constExponent=3}
  n22 = VectorParameter() {group=CarrotCake, parameterName=CenterGlow_Color, sortPriority=30, defaultValue=<rgba 1 0.295 0 1>}
  n23 = ComponentMask(None=@n22.RGB) {b=true, g=true, r=true}
  NamedRerouteDeclaration(None=Multiply(A=Multiply(A=@n21, B=@n23), B=ScalarParameter() {defaultValue=1, group=CarrotCake, parameterName=CenterGlow_Intensity})) {name=FaceGlow, nodeColor=<rgba 0.6172 1 0 1>}
 ]}
