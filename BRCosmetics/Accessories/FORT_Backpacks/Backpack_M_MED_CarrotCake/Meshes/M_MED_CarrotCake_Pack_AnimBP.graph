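# Animation Blueprint for the CarrotCake backpack cosmetic, which plays a reactive salute when an enemy player is eliminated.
# The Backpack States machine loops idle, cuts to the salute without crossfade, then returns automatically, feeding three additive bone-space adjustments of the pack bone.
# The aie.EnemyPlayerEliminated anim input event sets the trigger flag, which post-evaluation clears.
# An Animation Testing state machine cycles idles and salute but is not connected to the output.
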AnimBlueprint /BRCosmetics/Accessories/FORT_Backpacks/Backpack_M_MED_CarrotCake/Meshes/M_MED_CarrotCake_Pack_AnimBP
target_skeleton=&/M_MED_CarrotCake_Pack_Skeleton
graphs [
 AnimGraph {nodes [
   Input Pose()
   AnimGraphNode_StateMachine_1 = Backpack States() {is=title AnimGraphNode_StateMachine}
   n1 = "Transform (Modify) Bone - Bone: pack"!(ComponentPose=Local To Component(LocalPose=@AnimGraphNode_StateMachine_1), Alpha, Scale="0.9,0.9,0.9") -> (Pose) {scaleMode=BMM_Additive, scaleSpace=BCS_BoneSpace, alphaBoolBlend {blendOption=Linear}, boneToModify {boneName=pack}, scale=<0.9 0.9 0.9>}
   n2 = "Transform (Modify) Bone - Bone: pack"!(ComponentPose=@n1, Alpha, Translation:Vector="0,-2,0") -> (Pose) {
    translationMode=BMM_Additive
    translationSpace=BCS_BoneSpace
    alphaBoolBlend {blendOption=Linear}
    boneToModify {boneName=pack}
    translation {y=-2}
   }
   n3 = "Transform (Modify) Bone - Bone: pack"!(ComponentPose=@n2, Alpha, Rotation:Rotator="0,0,-4") -> (Pose) {
    rotationMode=BMM_Additive
    rotationSpace=BCS_BoneSpace
    alphaBoolBlend {blendOption=Linear}
    boneToModify {boneName=pack}
    rotation {roll=-4}
   }
   Output Pose(Result=Component To Local(ComponentPose=@n3)) {layerGroup=None}
   AnimGraphNode_StateMachine_0 = Animation Testing() {is=title AnimGraphNode_StateMachine}
  ]}
 Animation Testing {
  entry=Idle_1
  path=AnimGraph.AnimGraphNode_StateMachine_0.Animation Testing
  ids=id
  nodes [
   AnimStateNode_0 = Idle_1() {
    is=title AnimStateNode
    stateEntered {guid=#1}
    stateFullyBlended {guid=#2}
    stateLeft {guid=#3}
   }
   AnimStateNode_1 = Idle_2() {
    is=title AnimStateNode
    stateEntered {guid=#4}
    stateFullyBlended {guid=#5}
    stateLeft {guid=#6}
   }
   AnimStateNode_2 = ReactivitySalute() {
    is=title AnimStateNode
    stateEntered {guid=#7}
    stateFullyBlended {guid=#8}
    stateLeft {guid=#9}
   }
   AnimStateTransitionNode_2 = AnimStateTransitionNode() {
    bAutomaticRuleBasedOnSequencePlayerInState=true
    transitionEnd {guid=#10}
    transitionInterrupt {guid=#11}
    transitionStart {guid=#12}
   }
   AnimStateTransitionNode_3 = AnimStateTransitionNode() {
    bAutomaticRuleBasedOnSequencePlayerInState=true
    transitionEnd {guid=#13}
    transitionInterrupt {guid=#14}
    transitionStart {guid=#15}
   }
   AnimStateTransitionNode_4 = AnimStateTransitionNode() {bAutomaticRuleBasedOnSequencePlayerInState=true}
   AnimStateTransitionNode_5 = AnimStateTransitionNode() {bAutomaticRuleBasedOnSequencePlayerInState=true}
  ]
  links [
   {
    dst=Idle_2
    rule=AnimGraph.AnimGraphNode_StateMachine_0.Animation Testing.AnimStateTransitionNode_2.Transition
    src=Idle_1
    via=AnimStateTransitionNode_2
   }
   {
    dst=ReactivitySalute
    rule=AnimGraph.AnimGraphNode_StateMachine_0.Animation Testing.AnimStateTransitionNode_3.Transition
    src=Idle_2
    via=AnimStateTransitionNode_3
   }
   {
    dst=Idle_1
    rule=AnimGraph.AnimGraphNode_StateMachine_0.Animation Testing.AnimStateTransitionNode_4.Transition
    src=ReactivitySalute
    via=AnimStateTransitionNode_4
   }
   {
    dst=Idle_1
    rule=AnimGraph.AnimGraphNode_StateMachine_0.Animation Testing.AnimStateTransitionNode_5.Transition
    src=ReactivitySalute
    via=AnimStateTransitionNode_5
   }
  ]
 }
 Idle_1 {
  path=AnimGraph.AnimGraphNode_StateMachine_0.Animation Testing.AnimStateNode_0.Idle_1
  nodes [
   n1 = Sequence Player 'CarrotCake_Idle_Anim_PACK'() {bLoopAnimation=false, sequence=/BRCosmetics/Accessories/FORT_Backpacks/Backpack_M_MED_CarrotCake/Animation/CarrotCake_Idle_Anim_PACK}
   Output Animation Pose(Result=@n1)
  ]
 }
 Idle_2 {
  path=AnimGraph.AnimGraphNode_StateMachine_0.Animation Testing.AnimStateNode_1.Idle_2
  nodes [
   n1 = Sequence Player 'CarrotCake_Idle_Anim_PACK'() {bLoopAnimation=false, sequence=&CarrotCake_Idle_Anim_PACK}
   Output Animation Pose(Result=@n1)
  ]
 }
 ReactivitySalute {
  path=AnimGraph.AnimGraphNode_StateMachine_0.Animation Testing.AnimStateNode_2.ReactivitySalute
  nodes [
   "Sequence Player (additive) 'MuteRibbon_IdleBreak_Anim_PACK'"() {bLoopAnimation=false, sequence=/BRCosmetics/Accessories/FORT_Backpacks/Backpack_F_MED_MuteRibbon/Animation/MuteRibbon_IdleBreak_Anim_PACK}
   n1 = Sequence Player 'CarrotCake_ReactiveSalute_Anim_PACK'() {
    bLoopAnimation=false
    sequence=/BRCosmetics/Accessories/FORT_Backpacks/Backpack_M_MED_CarrotCake/Animation/CarrotCake_ReactiveSalute_Anim_PACK
   }
   Output Animation Pose(Result=@n1)
  ]
 }
 Transition {path=AnimGraph.AnimGraphNode_StateMachine_0.Animation Testing.AnimStateTransitionNode_2.Transition, nodes [Result()]}
 Transition {path=AnimGraph.AnimGraphNode_StateMachine_0.Animation Testing.AnimStateTransitionNode_3.Transition, nodes [Result()]}
 Transition {path=AnimGraph.AnimGraphNode_StateMachine_0.Animation Testing.AnimStateTransitionNode_4.Transition, nodes [Result()]}
 Transition {path=AnimGraph.AnimGraphNode_StateMachine_0.Animation Testing.AnimStateTransitionNode_5.Transition, nodes [Result()]}
 Backpack States {
  entry=Idle
  path=AnimGraph.AnimGraphNode_StateMachine_1.Backpack States
  ids=id
  nodes [
   AnimStateNode_0 = Idle() {stateEntered {guid=#16}}
   AnimStateNode_4 = Elimination() {is=title AnimStateNode}
   AnimStateTransitionNode_11 = AnimStateTransitionNode() {crossfadeDuration=0}
   AnimStateTransitionNode_5 = AnimStateTransitionNode() {bAutomaticRuleBasedOnSequencePlayerInState=true, crossfadeDuration=0}
  ]
  links [
   {
    dst=Idle
    rule=AnimGraph.AnimGraphNode_StateMachine_1.Backpack States.AnimStateTransitionNode_5.Transition
    src=Elimination
    via=AnimStateTransitionNode_5
   }
   {
    dst=Elimination
    rule=AnimGraph.AnimGraphNode_StateMachine_1.Backpack States.AnimStateTransitionNode_11.Transition
    src=Idle
    via=AnimStateTransitionNode_11
   }
  ]
 }
 Idle {
  path=AnimGraph.AnimGraphNode_StateMachine_1.Backpack States.AnimStateNode_0.Idle
  nodes [Output Animation Pose(Result=Sequence Player 'CarrotCake_Idle_Anim_PACK'() {sequence=&CarrotCake_Idle_Anim_PACK})]
 }
 Elimination {
  path=AnimGraph.AnimGraphNode_StateMachine_1.Backpack States.AnimStateNode_4.Elimination
  nodes [
   n1 = Sequence Player 'CarrotCake_ReactiveSalute_Anim_PACK'() {
    bLoopAnimation=false
    sequence=&CarrotCake_ReactiveSalute_Anim_PACK
   }
   Output Animation Pose(Result=@n1)
  ]
 }
 Transition {
  path=AnimGraph.AnimGraphNode_StateMachine_1.Backpack States.AnimStateTransitionNode_11.Transition
  nodes [Result(bCanEnterTransition=Get bEnemyPlayerEliminated() -> (bEnemyPlayerEliminated:bool))]
 }
 Transition {path=AnimGraph.AnimGraphNode_StateMachine_1.Backpack States.AnimStateTransitionNode_5.Transition, nodes [Result()]}
 EventGraph {nodes [
   Event BlueprintPostEvaluateAnimation() {is=title K2Node_Event}
   > Set bEnemyPlayerEliminated(bEnemyPlayerEliminated:bool=false) -> (Output_Get:bool)
   n1 = Event BlueprintProcessAnimInputEvent()
   > Branch(Condition=IsNamed(self=@n1.AnimInputEvent, InEventName=aie.EnemyPlayerEliminated))
   > Set bEnemyPlayerEliminated(bEnemyPlayerEliminated=true)
   TryGetPawnOwner()
   Event BlueprintUpdateAnimation()
  ]}
]
variables [bEnemyPlayerEliminated {type=boolean}]
